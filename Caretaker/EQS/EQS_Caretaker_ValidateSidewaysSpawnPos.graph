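# EQS query that validates spawn positions for the Caretaker's Sideways content.
# A composite generator combines the querier's current location with donut points around it.
# Filters require minimum distances from all cubes, Caretaker no-Sideways POIs and major water bodies, measured to actor bounds.
# Points must also project onto navmesh using a water-excluding filter and pass a wide-radius flat-surface test.
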EnvQuery /Caretaker/EQS/EQS_Caretaker_ValidateSidewaysSpawnPos
options [{
  generator {
   name=EnvQueryGenerator_Composite_0
   class=EnvQueryGenerator_Composite
   properties {
    optionName=""
    generators [
     &self:EnvQueryGenerator_Composite_0.EnvQueryGenerator_CurrentLocation_0
     &self:EnvQueryGenerator_Composite_0.EnvQueryGenerator_Donut_0
    ]
   }
  }
  tests [
   FortQueryTest_DistanceToActorBound_6 {
    class=FortQueryTest_DistanceToActorBound
    properties {
     bDefineReferenceValue=false
     bUse2DDistance=false
     boolValue=true
     clampMaxType=None
     clampMinType=None
     distanceTo=/SidewaysGameplay/Environmental/SidewaysEncounter/EnvQueryContext_AllCube.EnvQueryContext_AllCube_C
     filterType=Minimum
     floatValueMax=0
     floatValueMin=3000
     multipleContextFilterOp=AllPass
     multipleContextScoreOp=AverageScore
     normalizationType=Absolute
     referenceValue=0
     scoreClampMax=0
     scoreClampMin=0
     scoringEquation=Linear
     scoringFactor=1
     testComment=""
     testPurpose=Filter
    }
   }
   FortQueryTest_DistanceToActorBound_7 {
    class=FortQueryTest_DistanceToActorBound
    properties {
     bDefineReferenceValue=false
     bUse2DDistance=false
     boolValue=true
     clampMaxType=None
     clampMinType=None
     distanceTo=/Game/Athena/AI/EQS/Contexts/EnvQueryContext_CaretakerNoSidewaysPOI.EnvQueryContext_CaretakerNoSidewaysPOI_C
     filterType=Minimum
     floatValueMax=0
     floatValueMin=5000
     multipleContextFilterOp=AllPass
     multipleContextScoreOp=AverageScore
     normalizationType=Absolute
     referenceValue=0
     scoreClampMax=0
     scoreClampMin=0
     scoringEquation=Linear
     scoringFactor=1
     testComment=POI
     testPurpose=Filter
    }
   }
   FortQueryTest_DistanceToActorBound_8 {
    class=FortQueryTest_DistanceToActorBound
    properties {
     bDefineReferenceValue=false
     bUse2DDistance=false
     boolValue=true
     clampMaxType=None
     clampMinType=None
     distanceTo=/Game/Athena/AI/EQS/Contexts/EnvQueryContext_AllMajorWaterBodies.EnvQueryContext_AllMajorWaterBodies_C
     filterType=Minimum
     floatValueMax=0
     floatValueMin=3000
     multipleContextFilterOp=AllPass
     multipleContextScoreOp=AverageScore
     normalizationType=Absolute
     referenceValue=0
     scoreClampMax=0
     scoreClampMin=0
     scoringEquation=Linear
     scoringFactor=1
     testComment=""
     testPurpose=Filter
    }
   }
   EnvQueryTest_Project_1 {
    class=EnvQueryTest_Project
    properties {
     scoringFactor=1
     testComment=""
     testPurpose=Filter
     projectionData {
      navigationFilter=&NavFilter_NPC_ExcludeAllWater_C
      postProjectionVerticalOffset=0
      projectDown=1024
      projectUp=1024
      traceMode=Navigation
     }
    }
   }
   FortQueryTest_OnFlatSurface_2 {class=FortQueryTest_OnFlatSurface, properties {radius=2000, toleranceZ=500, traceOffsetDown=500, traceOffsetUp=500}}
  ]
 }]
