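# EQS query that picks a free-wander destination for the Caretaker AI on navmesh-projected circles.
# Four fallback options try a large then small circle with no-blocker pathfinding, then repeat both with avoid-all-water pathfinding.
# Scoring weights spawn proximity highest, clamped by difficulty curves, plus distance from self, closeness to friendly AI and players, facing and slight randomness.
# Filters reject points too close to nearby AI pawns or their move destinations.
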EnvQuery /Caretaker/EQS/EQS_Caretaker_WanderFreely
options [
 {
  generator {
   name=EnvQueryGenerator_OnCircle_1
   class=EnvQueryGenerator_OnCircle
   properties {
    numberOfPoints=16
    circleRadius {dataBinding=null, dataField=BehaviorDistance, defaultValue=4000}
    projectionData {
     extentX=10
     navigationFilter=/Caretaker/NavFilter/NavFilter_Caretaker_NoBlockers.NavFilter_Caretaker_NoBlockers_C
     postProjectionVerticalOffset=30
     projectDown=4096
    }
    traceData {extentX=0, navigationFilter=&NavFilter_Caretaker_NoBlockers_C}
   }
  }
  tests [
   EnvQueryTest_PathfindingBatch_0 {
    class=EnvQueryTest_PathfindingBatch
    properties {filterClass=&NavFilter_Caretaker_NoBlockers_C}
   }
   EnvQueryTest_Distance_0 {
    class=EnvQueryTest_Distance
    properties {
     distanceTo=&EnvQueryContext_Querier
     filterType=Range
     floatValueMin=0
     scoringEquation=Linear
     testComment=Pick point farthest from Self
     testMode=Distance2D
     testPurpose=Score
     scoringFactor {dataBinding=null, dataField=FloatValue, defaultValue=2}
    }
   }
   EnvQueryTest_Dot_0 {
    class=EnvQueryTest_Dot
    properties {
     clampMaxType=None
     clampMinType=None
     scoreClampMax=0
     scoreClampMin=0
     testComment=querier facing item
     scoringFactor {dataBinding=null, dataField=FloatValue, defaultValue=0.2}
    }
   }
   FortQueryTest_Random_0 {class=FortQueryTest_Random, properties {floatValueMax=0.1, scoringFactor {dataBinding=null, dataField=FloatValue, defaultValue=0.1}}}
   EnvQueryTest_Distance_1 {
    class=EnvQueryTest_Distance
    properties {
     clampMaxType=SpecifiedValue
     clampMinType=SpecifiedValue
     distanceTo=/Script/FortniteAIServer.FortQueryContext_AIPawnSpawnLocation
     filterType=Range
     floatValueMin=0
     scoringEquation=InverseLinear
     testComment=Ring around Spawn Location
     testMode=Distance2D
     testPurpose=Score
     scoreClampMax {
      dataBinding=&self:EnvQueryTest_Distance_1.FortAIDataProvider_FloatCurveOverGameDifficulty_1
      dataField=FloatValue
      defaultValue=15000
     }
     scoreClampMin {
      dataBinding=&self:EnvQueryTest_Distance_1.FortAIDataProvider_FloatCurveOverGameDifficulty_0
      dataField=FloatValue
      defaultValue=10000
     }
     scoringFactor {dataBinding=null, dataField=FloatValue, defaultValue=10}
    }
   }
   EnvQueryTest_Distance_2 {
    class=EnvQueryTest_Distance
    properties {
     distanceTo=&FortQueryContext_NearbyAIPawns
     filterType=Minimum
     floatValueMin=250
     scoringEquation=InverseLinear
     scoringFactor=3
     testComment=Try not to overlap AI Pawns
     testMode=Distance3D
     testPurpose=Filter
    }
   }
   EnvQueryTest_Distance_3 {
    class=EnvQueryTest_Distance
    properties {
     distanceTo=/Script/FortniteAIServer.FortQueryContext_NearbyFriendlyAIPawnsAndPlayerPawns
     filterType=Range
     floatValueMin=0
     scoringEquation=InverseLinear
     scoringFactor=3
     testComment=Stay near friends
     testMode=Distance3D
     testPurpose=Score
    }
   }
   EnvQueryTest_Distance_4 {
    class=EnvQueryTest_Distance
    properties {
     distanceTo=&FortQueryContext_NearbyAIPawnsMoveDestinations
     filterType=Minimum
     floatValueMin=250
     scoringEquation=Linear
     scoringFactor=1
     testComment=not near other pawn move destinations
     testMode=Distance3D
     testPurpose=Filter
    }
   }
  ]
 }
 {
  generator {
   name=EnvQueryGenerator_OnCircle_2
   class=EnvQueryGenerator_OnCircle
   properties {
    numberOfPoints=8
    circleRadius {dataBinding=null, dataField=BehaviorDistance, defaultValue=1000}
    projectionData {
     extentX=10
     navigationFilter=&NavFilter_Caretaker_NoBlockers_C
     postProjectionVerticalOffset=30
     projectDown=4096
    }
    traceData {extentX=0, navigationFilter=&NavFilter_Caretaker_NoBlockers_C}
   }
  }
  tests [
   EnvQueryTest_PathfindingBatch_2 {
    class=EnvQueryTest_PathfindingBatch
    properties {filterClass=&NavFilter_Caretaker_NoBlockers_C}
   }
   EnvQueryTest_Distance_5 {
    class=EnvQueryTest_Distance
    properties {
     distanceTo=&EnvQueryContext_Querier
     filterType=Range
     floatValueMin=0
     scoringEquation=Linear
     testComment=Pick point farthest from Self
     testMode=Distance2D
     testPurpose=Score
     scoringFactor {dataBinding=null, dataField=FloatValue, defaultValue=2}
    }
   }
   EnvQueryTest_Dot_1 {
    class=EnvQueryTest_Dot
    properties {
     clampMaxType=None
     clampMinType=None
     scoreClampMax=0
     scoreClampMin=0
     testComment=querier facing item
     scoringFactor {dataBinding=null, dataField=FloatValue, defaultValue=0.2}
    }
   }
   FortQueryTest_Random_1 {class=FortQueryTest_Random, properties {floatValueMax=0.1, scoringFactor {dataBinding=null, dataField=FloatValue, defaultValue=0.1}}}
   EnvQueryTest_Distance_20 {
    class=EnvQueryTest_Distance
    properties {
     clampMaxType=SpecifiedValue
     clampMinType=SpecifiedValue
     distanceTo=&FortQueryContext_AIPawnSpawnLocation
     filterType=Range
     floatValueMin=0
     scoringEquation=InverseLinear
     testComment=Ring around Spawn Location
     testMode=Distance2D
     testPurpose=Score
     scoreClampMax {
      dataBinding=&self:EnvQueryTest_Distance_20.FortAIDataProvider_FloatCurveOverGameDifficulty_1
      dataField=FloatValue
      defaultValue=15000
     }
     scoreClampMin {
      dataBinding=&self:EnvQueryTest_Distance_20.FortAIDataProvider_FloatCurveOverGameDifficulty_0
      dataField=FloatValue
      defaultValue=10000
     }
     scoringFactor {dataBinding=null, dataField=FloatValue, defaultValue=10}
    }
   }
   EnvQueryTest_Distance_12 {
    class=EnvQueryTest_Distance
    properties {
     distanceTo=&FortQueryContext_NearbyAIPawns
     filterType=Minimum
     floatValueMin=250
     scoringEquation=InverseLinear
     scoringFactor=3
     testComment=Try not to overlap AI Pawns
     testMode=Distance3D
     testPurpose=Filter
    }
   }
   EnvQueryTest_Distance_13 {
    class=EnvQueryTest_Distance
    properties {
     distanceTo=&FortQueryContext_NearbyFriendlyAIPawnsAndPlayerPawns
     filterType=Range
     floatValueMin=0
     scoringEquation=InverseLinear
     scoringFactor=3
     testComment=Stay near friends
     testMode=Distance3D
     testPurpose=Score
    }
   }
   EnvQueryTest_Distance_14 {
    class=EnvQueryTest_Distance
    properties {
     distanceTo=&FortQueryContext_NearbyAIPawnsMoveDestinations
     filterType=Minimum
     floatValueMin=250
     scoringEquation=Linear
     scoringFactor=1
     testComment=not near other pawn move destinations
     testMode=Distance3D
     testPurpose=Filter
    }
   }
  ]
 }
 {
  generator {
   name=EnvQueryGenerator_OnCircle_0
   class=EnvQueryGenerator_OnCircle
   properties {
    numberOfPoints=16
    circleRadius {dataBinding=null, dataField=BehaviorDistance, defaultValue=4000}
    projectionData {
     extentX=10
     navigationFilter=&NavFilter_Caretaker_NoBlockers_C
     postProjectionVerticalOffset=30
     projectDown=4096
    }
    traceData {extentX=0, navigationFilter=&NavFilter_Caretaker_NoBlockers_C}
   }
  }
  tests [
   EnvQueryTest_PathfindingBatch_3 {
    class=EnvQueryTest_PathfindingBatch
    properties {filterClass=&NavFilter_NPC_AvoidAllWater_C}
   }
   EnvQueryTest_Distance_15 {
    class=EnvQueryTest_Distance
    properties {
     distanceTo=&EnvQueryContext_Querier
     filterType=Range
     floatValueMin=0
     scoringEquation=Linear
     testComment=Pick point farthest from Self
     testMode=Distance2D
     testPurpose=Score
     scoringFactor {dataBinding=null, dataField=FloatValue, defaultValue=2}
    }
   }
   EnvQueryTest_Dot_3 {
    class=EnvQueryTest_Dot
    properties {
     clampMaxType=None
     clampMinType=None
     scoreClampMax=0
     scoreClampMin=0
     testComment=querier facing item
     scoringFactor {dataBinding=null, dataField=FloatValue, defaultValue=0.2}
    }
   }
   FortQueryTest_Random_3 {class=FortQueryTest_Random, properties {floatValueMax=0.1, scoringFactor {dataBinding=null, dataField=FloatValue, defaultValue=0.1}}}
   EnvQueryTest_Distance_21 {
    class=EnvQueryTest_Distance
    properties {
     clampMaxType=SpecifiedValue
     clampMinType=SpecifiedValue
     distanceTo=&FortQueryContext_AIPawnSpawnLocation
     filterType=Range
     floatValueMin=0
     scoringEquation=InverseLinear
     testComment=Ring around Spawn Location
     testMode=Distance2D
     testPurpose=Score
     scoreClampMax {
      dataBinding=&self:EnvQueryTest_Distance_21.FortAIDataProvider_FloatCurveOverGameDifficulty_1
      dataField=FloatValue
      defaultValue=15000
     }
     scoreClampMin {
      dataBinding=&self:EnvQueryTest_Distance_21.FortAIDataProvider_FloatCurveOverGameDifficulty_0
      dataField=FloatValue
      defaultValue=10000
     }
     scoringFactor {dataBinding=null, dataField=FloatValue, defaultValue=10}
    }
   }
   EnvQueryTest_Distance_17 {
    class=EnvQueryTest_Distance
    properties {
     distanceTo=&FortQueryContext_NearbyAIPawns
     filterType=Minimum
     floatValueMin=250
     scoringEquation=InverseLinear
     scoringFactor=3
     testComment=Try not to overlap AI Pawns
     testMode=Distance3D
     testPurpose=Filter
    }
   }
   EnvQueryTest_Distance_18 {
    class=EnvQueryTest_Distance
    properties {
     distanceTo=&FortQueryContext_NearbyFriendlyAIPawnsAndPlayerPawns
     filterType=Range
     floatValueMin=0
     scoringEquation=InverseLinear
     scoringFactor=3
     testComment=Stay near friends
     testMode=Distance3D
     testPurpose=Score
    }
   }
   EnvQueryTest_Distance_19 {
    class=EnvQueryTest_Distance
    properties {
     distanceTo=&FortQueryContext_NearbyAIPawnsMoveDestinations
     filterType=Minimum
     floatValueMin=250
     scoringEquation=Linear
     scoringFactor=1
     testComment=not near other pawn move destinations
     testMode=Distance3D
     testPurpose=Filter
    }
   }
  ]
 }
 {
  generator {
   name=EnvQueryGenerator_OnCircle_3
   class=EnvQueryGenerator_OnCircle
   properties {
    numberOfPoints=8
    circleRadius {dataBinding=null, dataField=BehaviorDistance, defaultValue=1000}
    projectionData {
     extentX=10
     navigationFilter=&NavFilter_Caretaker_NoBlockers_C
     postProjectionVerticalOffset=30
     projectDown=4096
    }
    traceData {extentX=0, navigationFilter=&NavFilter_Caretaker_NoBlockers_C}
   }
  }
  tests [
   EnvQueryTest_PathfindingBatch_1 {
    class=EnvQueryTest_PathfindingBatch
    properties {filterClass=&NavFilter_NPC_AvoidAllWater_C}
   }
   EnvQueryTest_Distance_7 {
    class=EnvQueryTest_Distance
    properties {
     distanceTo=&EnvQueryContext_Querier
     filterType=Range
     floatValueMin=0
     scoringEquation=Linear
     testComment=Pick point farthest from Self
     testMode=Distance2D
     testPurpose=Score
     scoringFactor {dataBinding=null, dataField=FloatValue, defaultValue=2}
    }
   }
   EnvQueryTest_Dot_2 {
    class=EnvQueryTest_Dot
    properties {
     clampMaxType=None
     clampMinType=None
     scoreClampMax=0
     scoreClampMin=0
     testComment=querier facing item
     scoringFactor {dataBinding=null, dataField=FloatValue, defaultValue=0.2}
    }
   }
   FortQueryTest_Random_2 {class=FortQueryTest_Random, properties {floatValueMax=0.1, scoringFactor {dataBinding=null, dataField=FloatValue, defaultValue=0.1}}}
   EnvQueryTest_Distance_22 {
    class=EnvQueryTest_Distance
    properties {
     clampMaxType=SpecifiedValue
     clampMinType=SpecifiedValue
     distanceTo=&FortQueryContext_AIPawnSpawnLocation
     filterType=Range
     floatValueMin=0
     scoringEquation=InverseLinear
     testComment=Ring around Spawn Location
     testMode=Distance2D
     testPurpose=Score
     scoreClampMax {
      dataBinding=&self:EnvQueryTest_Distance_22.FortAIDataProvider_FloatCurveOverGameDifficulty_1
      dataField=FloatValue
      defaultValue=15000
     }
     scoreClampMin {
      dataBinding=&self:EnvQueryTest_Distance_22.FortAIDataProvider_FloatCurveOverGameDifficulty_0
      dataField=FloatValue
      defaultValue=10000
     }
     scoringFactor {dataBinding=null, dataField=FloatValue, defaultValue=10}
    }
   }
   EnvQueryTest_Distance_9 {
    class=EnvQueryTest_Distance
    properties {
     distanceTo=&FortQueryContext_NearbyAIPawns
     filterType=Minimum
     floatValueMin=250
     scoringEquation=InverseLinear
     scoringFactor=3
     testComment=Try not to overlap AI Pawns
     testMode=Distance3D
     testPurpose=Filter
    }
   }
   EnvQueryTest_Distance_11 {
    class=EnvQueryTest_Distance
    properties {
     distanceTo=&FortQueryContext_NearbyFriendlyAIPawnsAndPlayerPawns
     filterType=Range
     floatValueMin=0
     scoringEquation=InverseLinear
     scoringFactor=3
     testComment=Stay near friends
     testMode=Distance3D
     testPurpose=Score
    }
   }
   EnvQueryTest_Distance_10 {
    class=EnvQueryTest_Distance
    properties {
     distanceTo=&FortQueryContext_NearbyAIPawnsMoveDestinations
     filterType=Minimum
     floatValueMin=250
     scoringEquation=Linear
     scoringFactor=1
     testComment=not near other pawn move destinations
     testMode=Distance3D
     testPurpose=Filter
    }
   }
  ]
 }
]
