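# EQS query that picks a wander point for the Caretaker AI near its goal, using three circle-generator options in fallback order.
# Each projects points onto a no-blockers navmesh, tests pathfinding, favors forward but not straight-ahead points, and rejects those near itself, other AI or their destinations.
# The first two, wide then tight circles, keep a minimum goal distance while scoring closer points higher.
# The last returns wide, replacing that with a goal-direction preference.
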EnvQuery /Caretaker/EQS/EQS_Caretaker_WanderNearGoal
options [
 {
  generator {
   name=EnvQueryGenerator_OnCircle_1
   class=EnvQueryGenerator_OnCircle
   properties {
    numberOfPoints=12
    circleRadius {dataBinding=null, dataField=BehaviorDistance, defaultValue=3000}
    projectionData {
     extentX=10
     navigationFilter=/Caretaker/NavFilter/NavFilter_Caretaker_NoBlockers.NavFilter_Caretaker_NoBlockers_C
     postProjectionVerticalOffset=30
     projectDown=4096
    }
    traceData {extentX=0, navigationFilter=&NavFilter_Caretaker_NoBlockers_C}
   }
  }
  tests [
   EnvQueryTest_PathfindingBatch_1 {
    class=EnvQueryTest_PathfindingBatch
    properties {filterClass=&NavFilter_Caretaker_NoBlockers_C}
   }
   EnvQueryTest_Distance_3 {
    class=EnvQueryTest_Distance
    properties {
     distanceTo=&EnvQueryContext_Querier
     filterType=Minimum
     scoringEquation=InverseLinear
     scoringFactor=1
     testComment=Throw out point I'm right on top of
     testMode=Distance2D
     testPurpose=Filter
     floatValueMin {
      dataBinding=&self:EnvQueryTest_Distance_3.FortAIDataProvider_FloatCurveOverGameDifficulty_0
      dataField=FloatValue
      defaultValue=200
     }
    }
   }
   EnvQueryTest_Dot_3 {
    class=EnvQueryTest_Dot
    properties {
     clampMaxType=SpecifiedValue
     clampMinType=SpecifiedValue
     referenceValue=0.707
     testComment=Prefer in front of Self
     scoreClampMax {dataBinding=null, dataField=FloatValue, defaultValue=0.4}
     scoreClampMin {dataBinding=null, dataField=FloatValue, defaultValue=0}
     scoringFactor {dataBinding=null, dataField=FloatValue, defaultValue=0.5}
    }
   }
   EnvQueryTest_Dot_6 {
    class=EnvQueryTest_Dot
    properties {
     clampMaxType=SpecifiedValue
     clampMinType=SpecifiedValue
     floatValueMax=1
     floatValueMin=0.95
     referenceValue=0.707
     scoringEquation=InverseLinear
     testComment=But not DIRECTLY in front of self
     scoreClampMax {dataBinding=null, dataField=FloatValue, defaultValue=1}
     scoreClampMin {dataBinding=null, dataField=FloatValue, defaultValue=0.9}
     scoringFactor {dataBinding=null, dataField=FloatValue, defaultValue=0.5}
    }
   }
   EnvQueryTest_Distance_8 {
    class=EnvQueryTest_Distance
    properties {
     clampMinType=SpecifiedValue
     distanceTo=&FortQueryContext_Goal
     filterType=Range
     floatValueMax=99999997952
     floatValueMin=3000
     scoreClampMin=5000
     scoringEquation=InverseLinear
     scoringFactor=1
     testComment="Get close, but not too close"
     testMode=Distance2D
     testPurpose=FilterAndScore
    }
   }
   EnvQueryTest_Distance_9 {
    class=EnvQueryTest_Distance
    properties {
     distanceTo=&FortQueryContext_NearbyAIPawns
     filterType=Minimum
     floatValueMin=250
     scoringEquation=Linear
     scoringFactor=1
     testComment=Throw out points near other AI
     testMode=Distance3D
     testPurpose=Filter
    }
   }
   EnvQueryTest_Distance_10 {
    class=EnvQueryTest_Distance
    properties {
     distanceTo=&FortQueryContext_NearbyAIPawnsMoveDestinations
     filterType=Minimum
     floatValueMin=250
     scoringEquation=Linear
     scoringFactor=1
     testComment=Throw out points near other AI's Movement Destinations
     testMode=Distance3D
     testPurpose=Filter
    }
   }
  ]
 }
 {
  generator {
   name=EnvQueryGenerator_OnCircle_2
   class=EnvQueryGenerator_OnCircle
   properties {
    numberOfPoints=6
    circleRadius {dataBinding=null, dataField=BehaviorDistance, defaultValue=1000}
    projectionData {
     extentX=10
     navigationFilter=&NavFilter_Caretaker_NoBlockers_C
     postProjectionVerticalOffset=30
     projectDown=4096
    }
    traceData {extentX=0, navigationFilter=&NavFilter_Caretaker_NoBlockers_C}
   }
  }
  tests [
   EnvQueryTest_PathfindingBatch_0 {
    class=EnvQueryTest_PathfindingBatch
    properties {filterClass=&NavFilter_Caretaker_NoBlockers_C}
   }
   EnvQueryTest_Distance_1 {
    class=EnvQueryTest_Distance
    properties {
     distanceTo=&EnvQueryContext_Querier
     filterType=Minimum
     scoringEquation=InverseLinear
     scoringFactor=1
     testComment=Throw out point I'm right on top of
     testMode=Distance2D
     testPurpose=Filter
     floatValueMin {
      dataBinding=&self:EnvQueryTest_Distance_1.FortAIDataProvider_FloatCurveOverGameDifficulty_0
      dataField=FloatValue
      defaultValue=200
     }
    }
   }
   EnvQueryTest_Dot_0 {
    class=EnvQueryTest_Dot
    properties {
     clampMaxType=SpecifiedValue
     clampMinType=SpecifiedValue
     referenceValue=0.707
     testComment=Prefer in front of Self
     scoreClampMax {dataBinding=null, dataField=FloatValue, defaultValue=0.4}
     scoreClampMin {dataBinding=null, dataField=FloatValue, defaultValue=0}
     scoringFactor {dataBinding=null, dataField=FloatValue, defaultValue=0.5}
    }
   }
   EnvQueryTest_Dot_1 {
    class=EnvQueryTest_Dot
    properties {
     clampMaxType=SpecifiedValue
     clampMinType=SpecifiedValue
     floatValueMax=1
     floatValueMin=0.95
     referenceValue=0.707
     scoringEquation=InverseLinear
     testComment=But not DIRECTLY in front of self
     scoreClampMax {dataBinding=null, dataField=FloatValue, defaultValue=1}
     scoreClampMin {dataBinding=null, dataField=FloatValue, defaultValue=0.9}
     scoringFactor {dataBinding=null, dataField=FloatValue, defaultValue=0.5}
    }
   }
   EnvQueryTest_Distance_4 {
    class=EnvQueryTest_Distance
    properties {
     clampMinType=SpecifiedValue
     distanceTo=&FortQueryContext_Goal
     filterType=Range
     floatValueMax=99999997952
     floatValueMin=3000
     scoreClampMin=5000
     scoringEquation=InverseLinear
     scoringFactor=1
     testComment="Get close, but not too close"
     testMode=Distance2D
     testPurpose=FilterAndScore
    }
   }
   EnvQueryTest_Distance_5 {
    class=EnvQueryTest_Distance
    properties {
     distanceTo=&FortQueryContext_NearbyAIPawns
     filterType=Minimum
     floatValueMin=250
     scoringEquation=Linear
     scoringFactor=1
     testComment=Throw out points near other AI
     testMode=Distance3D
     testPurpose=Filter
    }
   }
   EnvQueryTest_Distance_6 {
    class=EnvQueryTest_Distance
    properties {
     distanceTo=&FortQueryContext_NearbyAIPawnsMoveDestinations
     filterType=Minimum
     floatValueMin=250
     scoringEquation=Linear
     scoringFactor=1
     testComment=Throw out points near other AI's Movement Destinations
     testMode=Distance3D
     testPurpose=Filter
    }
   }
  ]
 }
 {
  generator {
   name=EnvQueryGenerator_OnCircle_0
   class=EnvQueryGenerator_OnCircle
   properties {
    numberOfPoints=12
    circleRadius {dataBinding=null, dataField=BehaviorDistance, defaultValue=3000}
    projectionData {
     extentX=10
     navigationFilter=&NavFilter_Caretaker_NoBlockers_C
     postProjectionVerticalOffset=30
     projectDown=4096
    }
    traceData {extentX=0, navigationFilter=&NavFilter_Caretaker_NoBlockers_C}
   }
  }
  tests [
   EnvQueryTest_PathfindingBatch_2 {
    class=EnvQueryTest_PathfindingBatch
    properties {filterClass=&NavFilter_Caretaker_NoBlockers_C}
   }
   EnvQueryTest_Distance_7 {
    class=EnvQueryTest_Distance
    properties {
     distanceTo=&EnvQueryContext_Querier
     filterType=Minimum
     scoringEquation=InverseLinear
     scoringFactor=1
     testComment=Throw out point I'm right on top of
     testMode=Distance2D
     testPurpose=Filter
     floatValueMin {
      dataBinding=&self:EnvQueryTest_Distance_7.FortAIDataProvider_FloatCurveOverGameDifficulty_0
      dataField=FloatValue
      defaultValue=200
     }
    }
   }
   EnvQueryTest_Dot_4 {
    class=EnvQueryTest_Dot
    properties {
     clampMaxType=SpecifiedValue
     clampMinType=SpecifiedValue
     referenceValue=0.707
     testComment=Prefer in front of Self
     scoreClampMax {dataBinding=null, dataField=FloatValue, defaultValue=0.4}
     scoreClampMin {dataBinding=null, dataField=FloatValue, defaultValue=0}
     scoringFactor {dataBinding=null, dataField=FloatValue, defaultValue=0.25}
    }
   }
   EnvQueryTest_Dot_5 {
    class=EnvQueryTest_Dot
    properties {
     clampMaxType=SpecifiedValue
     clampMinType=SpecifiedValue
     floatValueMax=1
     floatValueMin=0.95
     referenceValue=0.707
     scoringEquation=InverseLinear
     testComment=But not DIRECTLY in front of self
     scoreClampMax {dataBinding=null, dataField=FloatValue, defaultValue=1}
     scoreClampMin {dataBinding=null, dataField=FloatValue, defaultValue=0.9}
     scoringFactor {dataBinding=null, dataField=FloatValue, defaultValue=0.25}
    }
   }
   EnvQueryTest_Dot_2 {
    class=EnvQueryTest_Dot
    properties {
     clampMaxType=SpecifiedValue
     clampMinType=SpecifiedValue
     testComment=Prefer in direction of Goal
     lineA {dirMode=TwoPoints, lineFrom=&EnvQueryContext_Querier, lineTo=&EnvQueryContext_Item}
     lineB {lineTo=&FortQueryContext_Goal}
     scoreClampMax {dataBinding=null, dataField=FloatValue, defaultValue=1}
     scoreClampMin {dataBinding=null, dataField=FloatValue, defaultValue=-1}
     scoringFactor {dataBinding=null, dataField=FloatValue, defaultValue=0.5}
    }
   }
   EnvQueryTest_Distance_2 {
    class=EnvQueryTest_Distance
    properties {
     distanceTo=&FortQueryContext_NearbyAIPawns
     filterType=Minimum
     floatValueMin=250
     scoringEquation=Linear
     scoringFactor=1
     testComment=Throw out points near other AI
     testMode=Distance3D
     testPurpose=Filter
    }
   }
   EnvQueryTest_Distance_0 {
    class=EnvQueryTest_Distance
    properties {
     distanceTo=&FortQueryContext_NearbyAIPawnsMoveDestinations
     filterType=Minimum
     floatValueMin=250
     scoringEquation=Linear
     scoringFactor=1
     testComment=Throw out points near other AI's Movement Destinations
     testMode=Distance3D
     testPurpose=Filter
    }
   }
  ]
 }
]
