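# EQS query that picks a Caretaker wander point near, but outside a minimum distance from, its goal.
# Three fallback options project circles onto NoBlockers navmesh: wide, tight, then wide again.
# All test paths, reject points near itself, other AI or their destinations, and favor forward but not straight ahead.
# The first two require and score goal distance; the last drops that, favoring the goal's direction.
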EnvQuery /Caretaker/EQS/EQS_Caretaker_WanderNearGoal_Outside
options [
 {
  generator {
   name=EnvQueryGenerator_OnCircle_1
   class=EnvQueryGenerator_OnCircle
   properties {
    numberOfPoints=12
    circleRadius {dataBinding=null, dataField=BehaviorDistance, defaultValue=3000}
    projectionData {
     extentX=10
     navigationFilter=/Caretaker/NavFilter/NavFilter_Caretaker_NoBlockers.NavFilter_Caretaker_NoBlockers_C
     postProjectionVerticalOffset=30
     projectDown=4096
    }
    traceData {extentX=0, navigationFilter=&NavFilter_Caretaker_NoBlockers_C}
   }
  }
  tests [
   EnvQueryTest_PathfindingBatch_3 {
    class=EnvQueryTest_PathfindingBatch
    properties {filterClass=&NavFilter_Caretaker_NoBlockers_C}
   }
   EnvQueryTest_Distance_11 {
    class=EnvQueryTest_Distance
    properties {
     distanceTo=&EnvQueryContext_Querier
     filterType=Minimum
     scoringEquation=InverseLinear
     scoringFactor=1
     testComment=Throw out point I'm right on top of
     testMode=Distance2D
     testPurpose=Filter
     floatValueMin {
      dataBinding=&self:EnvQueryTest_Distance_11.FortAIDataProvider_FloatCurveOverGameDifficulty_0
      dataField=FloatValue
      defaultValue=200
     }
    }
   }
   EnvQueryTest_Dot_7 {
    class=EnvQueryTest_Dot
    properties {
     clampMaxType=SpecifiedValue
     clampMinType=SpecifiedValue
     referenceValue=0.707
     testComment=Prefer in front of Self
     scoreClampMax {dataBinding=null, dataField=FloatValue, defaultValue=0.4}
     scoreClampMin {dataBinding=null, dataField=FloatValue, defaultValue=0}
     scoringFactor {dataBinding=null, dataField=FloatValue, defaultValue=0.5}
    }
   }
   EnvQueryTest_Dot_8 {
    class=EnvQueryTest_Dot
    properties {
     clampMaxType=SpecifiedValue
     clampMinType=SpecifiedValue
     floatValueMax=1
     floatValueMin=0.95
     referenceValue=0.707
     scoringEquation=InverseLinear
     testComment=But not DIRECTLY in front of self
     scoreClampMax {dataBinding=null, dataField=FloatValue, defaultValue=1}
     scoreClampMin {dataBinding=null, dataField=FloatValue, defaultValue=0.9}
     scoringFactor {dataBinding=null, dataField=FloatValue, defaultValue=0.5}
    }
   }
   EnvQueryTest_Distance_12 {
    class=EnvQueryTest_Distance
    properties {
     clampMinType=SpecifiedValue
     distanceTo=&FortQueryContext_Goal
     filterType=Range
     floatValueMax=99999997952
     floatValueMin=3000
     scoreClampMin=5000
     scoringEquation=InverseLinear
     scoringFactor=1
     testComment="Get close, but not too close"
     testMode=Distance2D
     testPurpose=FilterAndScore
    }
   }
   EnvQueryTest_Distance_13 {
    class=EnvQueryTest_Distance
    properties {
     distanceTo=&FortQueryContext_NearbyAIPawns
     filterType=Minimum
     floatValueMin=250
     scoringEquation=Linear
     scoringFactor=1
     testComment=Throw out points near other AI
     testMode=Distance3D
     testPurpose=Filter
    }
   }
   EnvQueryTest_Distance_14 {
    class=EnvQueryTest_Distance
    properties {
     distanceTo=&FortQueryContext_NearbyAIPawnsMoveDestinations
     filterType=Minimum
     floatValueMin=250
     scoringEquation=Linear
     scoringFactor=1
     testComment=Throw out points near other AI's Movement Destinations
     testMode=Distance3D
     testPurpose=Filter
    }
   }
  ]
 }
 {
  generator {
   name=EnvQueryGenerator_OnCircle_2
   class=EnvQueryGenerator_OnCircle
   properties {
    numberOfPoints=6
    circleRadius {dataBinding=null, dataField=BehaviorDistance, defaultValue=1000}
    projectionData {
     extentX=10
     navigationFilter=&NavFilter_Caretaker_NoBlockers_C
     postProjectionVerticalOffset=30
     projectDown=4096
    }
    traceData {extentX=0, navigationFilter=&NavFilter_Caretaker_NoBlockers_C}
   }
  }
  tests [
   EnvQueryTest_PathfindingBatch_5 {
    class=EnvQueryTest_PathfindingBatch
    properties {filterClass=&NavFilter_Caretaker_NoBlockers_C}
   }
   EnvQueryTest_Distance_18 {
    class=EnvQueryTest_Distance
    properties {
     distanceTo=&EnvQueryContext_Querier
     filterType=Minimum
     scoringEquation=InverseLinear
     scoringFactor=1
     testComment=Throw out point I'm right on top of
     testMode=Distance2D
     testPurpose=Filter
     floatValueMin {
      dataBinding=&self:EnvQueryTest_Distance_18.FortAIDataProvider_FloatCurveOverGameDifficulty_0
      dataField=FloatValue
      defaultValue=200
     }
    }
   }
   EnvQueryTest_Dot_12 {
    class=EnvQueryTest_Dot
    properties {
     clampMaxType=SpecifiedValue
     clampMinType=SpecifiedValue
     referenceValue=0.707
     testComment=Prefer in front of Self
     scoreClampMax {dataBinding=null, dataField=FloatValue, defaultValue=0.4}
     scoreClampMin {dataBinding=null, dataField=FloatValue, defaultValue=0}
     scoringFactor {dataBinding=null, dataField=FloatValue, defaultValue=0.5}
    }
   }
   EnvQueryTest_Dot_13 {
    class=EnvQueryTest_Dot
    properties {
     clampMaxType=SpecifiedValue
     clampMinType=SpecifiedValue
     floatValueMax=1
     floatValueMin=0.95
     referenceValue=0.707
     scoringEquation=InverseLinear
     testComment=But not DIRECTLY in front of self
     scoreClampMax {dataBinding=null, dataField=FloatValue, defaultValue=1}
     scoreClampMin {dataBinding=null, dataField=FloatValue, defaultValue=0.9}
     scoringFactor {dataBinding=null, dataField=FloatValue, defaultValue=0.5}
    }
   }
   EnvQueryTest_Distance_19 {
    class=EnvQueryTest_Distance
    properties {
     clampMinType=SpecifiedValue
     distanceTo=&FortQueryContext_Goal
     filterType=Range
     floatValueMax=99999997952
     floatValueMin=3000
     scoreClampMin=5000
     scoringEquation=InverseLinear
     scoringFactor=1
     testComment="Get close, but not too close"
     testMode=Distance2D
     testPurpose=FilterAndScore
    }
   }
   EnvQueryTest_Distance_20 {
    class=EnvQueryTest_Distance
    properties {
     distanceTo=&FortQueryContext_NearbyAIPawns
     filterType=Minimum
     floatValueMin=250
     scoringEquation=Linear
     scoringFactor=1
     testComment=Throw out points near other AI
     testMode=Distance3D
     testPurpose=Filter
    }
   }
   EnvQueryTest_Distance_21 {
    class=EnvQueryTest_Distance
    properties {
     distanceTo=&FortQueryContext_NearbyAIPawnsMoveDestinations
     filterType=Minimum
     floatValueMin=250
     scoringEquation=Linear
     scoringFactor=1
     testComment=Throw out points near other AI's Movement Destinations
     testMode=Distance3D
     testPurpose=Filter
    }
   }
  ]
 }
 {
  generator {
   name=EnvQueryGenerator_OnCircle_0
   class=EnvQueryGenerator_OnCircle
   properties {
    numberOfPoints=12
    circleRadius {dataBinding=null, dataField=BehaviorDistance, defaultValue=3000}
    projectionData {
     extentX=10
     navigationFilter=&NavFilter_Caretaker_NoBlockers_C
     postProjectionVerticalOffset=30
     projectDown=4096
    }
    traceData {extentX=0, navigationFilter=&NavFilter_Caretaker_NoBlockers_C}
   }
  }
  tests [
   EnvQueryTest_PathfindingBatch_4 {
    class=EnvQueryTest_PathfindingBatch
    properties {filterClass=&NavFilter_Caretaker_NoBlockers_C}
   }
   EnvQueryTest_Distance_15 {
    class=EnvQueryTest_Distance
    properties {
     distanceTo=&EnvQueryContext_Querier
     filterType=Minimum
     scoringEquation=InverseLinear
     scoringFactor=1
     testComment=Throw out point I'm right on top of
     testMode=Distance2D
     testPurpose=Filter
     floatValueMin {
      dataBinding=&self:EnvQueryTest_Distance_15.FortAIDataProvider_FloatCurveOverGameDifficulty_0
      dataField=FloatValue
      defaultValue=200
     }
    }
   }
   EnvQueryTest_Dot_9 {
    class=EnvQueryTest_Dot
    properties {
     clampMaxType=SpecifiedValue
     clampMinType=SpecifiedValue
     referenceValue=0.707
     testComment=Prefer in front of Self
     scoreClampMax {dataBinding=null, dataField=FloatValue, defaultValue=0.4}
     scoreClampMin {dataBinding=null, dataField=FloatValue, defaultValue=0}
     scoringFactor {dataBinding=null, dataField=FloatValue, defaultValue=0.25}
    }
   }
   EnvQueryTest_Dot_10 {
    class=EnvQueryTest_Dot
    properties {
     clampMaxType=SpecifiedValue
     clampMinType=SpecifiedValue
     floatValueMax=1
     floatValueMin=0.95
     referenceValue=0.707
     scoringEquation=InverseLinear
     testComment=But not DIRECTLY in front of self
     scoreClampMax {dataBinding=null, dataField=FloatValue, defaultValue=1}
     scoreClampMin {dataBinding=null, dataField=FloatValue, defaultValue=0.9}
     scoringFactor {dataBinding=null, dataField=FloatValue, defaultValue=0.25}
    }
   }
   EnvQueryTest_Dot_11 {
    class=EnvQueryTest_Dot
    properties {
     clampMaxType=SpecifiedValue
     clampMinType=SpecifiedValue
     testComment=Prefer in direction of Goal
     lineA {dirMode=TwoPoints, lineFrom=&EnvQueryContext_Querier, lineTo=&EnvQueryContext_Item}
     lineB {lineTo=&FortQueryContext_Goal}
     scoreClampMax {dataBinding=null, dataField=FloatValue, defaultValue=1}
     scoreClampMin {dataBinding=null, dataField=FloatValue, defaultValue=-1}
     scoringFactor {dataBinding=null, dataField=FloatValue, defaultValue=0.5}
    }
   }
   EnvQueryTest_Distance_16 {
    class=EnvQueryTest_Distance
    properties {
     distanceTo=&FortQueryContext_NearbyAIPawns
     filterType=Minimum
     floatValueMin=250
     scoringEquation=Linear
     scoringFactor=1
     testComment=Throw out points near other AI
     testMode=Distance3D
     testPurpose=Filter
    }
   }
   EnvQueryTest_Distance_17 {
    class=EnvQueryTest_Distance
    properties {
     distanceTo=&FortQueryContext_NearbyAIPawnsMoveDestinations
     filterType=Minimum
     floatValueMin=250
     scoringEquation=Linear
     scoringFactor=1
     testComment=Throw out points near other AI's Movement Destinations
     testMode=Distance3D
     testPurpose=Filter
    }
   }
  ]
 }
]
