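; An EQS query context for the Caretaker that supplies the querier's feet position as a single location.
; It casts the querier actor to Character, takes the actor location, and lowers Z by the capsule's scaled half height.
; The result returns through a merge-execution-with-data node, with no alternative location computed for a failed cast.
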(fn ProvideSingleLocation
  (params (QuerierObject :type Object%20Reference) (QuerierActor :type Actor%20Object%20Reference))
  (bind _ascharacter (Utilities|Casting|CastToCharacter QuerierActor)
    (:then
      _)
    (:CastFailed
      _))
  (bind _data (Utilities|FlowControl|MergeExecutionwithData (Math|Vector|MakeVector (.x (Transformation|GetActorLocation _ascharacter)) (.y (Transformation|GetActorLocation _ascharacter)) (- (.z (Transformation|GetActorLocation _ascharacter)) (Components|Capsule|GetScaledCapsuleHalfHeight (Variables|/Script/Engine.Character|GetCapsuleComponent _ascharacter)))))
    (:Then
      (return _data))))
